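# First-person Sound Cue for firing the break-action shotgun.
# A Random node picks among six equally weighted shot waves, and a Modulator above it applies a randomized volume boost to vary each shot.
# It references shared first-person weapon attenuation, a ranged-fire 1P sound class and a first-person gunfire concurrency set.
# The cue is saved with its debug flag enabled.
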SoundCue /ResolveGameplay/Items/Guns/BreakActionShotgun/Audio/BreakActionShotgun_Shoot_1P_Cue
root=Modulator_0
nodes [
 {id=Modulator_0, children [Random_1], properties {volumeMax=1.8, volumeMin=1.6}}
 {id=Random_1, children [WavePlayer_1,WavePlayer_2,WavePlayer_3,WavePlayer_4,WavePlayer_8,WavePlayer_9], properties {weights [1,1,1,1,1,1]}}
 {id=WavePlayer_1, properties {soundWaveAssetPtr=&/BreakActionShotgun_Shoot_01}}
 {id=WavePlayer_2, properties {soundWaveAssetPtr=&/BreakActionShotgun_Shoot_02}}
 {id=WavePlayer_3, properties {soundWaveAssetPtr=&/BreakActionShotgun_Shoot_03}}
 {id=WavePlayer_4, properties {soundWaveAssetPtr=&/BreakActionShotgun_Shoot_04}}
 {id=WavePlayer_8, properties {soundWaveAssetPtr=&/BreakActionShotgun_Shoot_05}}
 {id=WavePlayer_9, properties {soundWaveAssetPtr=&/BreakActionShotgun_Shoot_06}}
]
properties {
 attenuationSettings=&Fort_Weapon_Projectile_FirstPerson_Athena_Att
 bDebug=true
 maxDistance=2097152
 soundClassObject=&Fort_WeaponRanged_Fire_1P_PSM
 volumeMultiplier=0.8
 attenuationOverrides {manualReverbSendLevel=-1.0165, reverbSendMethod=Manual}
 concurrencyOverrides {bLimitToOwner=true, maxCount=3, resolutionRule=StopOldest, volumeScale=0.9}
 concurrencySet [&Weapon_Gunfire_FirstPerson_Conc]
}
