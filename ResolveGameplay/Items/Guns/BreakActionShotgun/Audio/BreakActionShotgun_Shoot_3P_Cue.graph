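# Third-person fire SoundCue for the break-action shotgun, layering close, medium and distant gunshot sounds.
# Each layer randomly picks one of six waves, applies a modulator, then its own distance attenuation and sound class, with a separate class for the distant layer.
# A mixer blends the three layers at decreasing input volumes, with the distant layer very low.
# Playback is limited by local and global gunfire concurrency, and debug is enabled.
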SoundCue /ResolveGameplay/Items/Guns/BreakActionShotgun/Audio/BreakActionShotgun_Shoot_3P_Cue
root=Mixer_17
nodes [
 {
  id=Attenuation_0
  children [Modulator_2]
  properties {attenuationSettings=&Fort_Weapon_Projectile_Medium_Mono_Athena_Att}
 }
 {
  id=Attenuation_34
  children [Modulator_0]
  properties {attenuationSettings=&Fort_Weapon_Projectile_Distant_Mono_Athena_Att}
 }
 {id=Attenuation_49, children [Modulator_1], properties {attenuationSettings=&Fort_Weapon_Projectile_Close_Athena_Att}}
 {id=Mixer_17, children [SoundClass_32, SoundClass_0, SoundClass_35], properties {inputVolume [0.8,0.6,0.08]}}
 {id=Modulator_0, children [Random_2], properties {pitchMax=1, pitchMin=1, volumeMax=2, volumeMin=2}}
 {id=Modulator_1, children [Random_0], properties {pitchMax=1.2, pitchMin=1, volumeMax=1, volumeMin=1}}
 {id=Modulator_2, children [Random_1], properties {pitchMax=1, pitchMin=1, volumeMax=1.95, volumeMin=1.9}}
 {id=Random_0, children [WavePlayer_3,WavePlayer_4,WavePlayer_5,WavePlayer_6,WavePlayer_7,WavePlayer_8], properties {weights [1,1,1,1,1,1]}}
 {
  id=Random_1
  children [WavePlayer_0, WavePlayer_1, WavePlayer_2, WavePlayer_9, WavePlayer_10, WavePlayer_11]
  properties {weights [1,1,1,1,1,1]}
 }
 {
  id=Random_2
  children [WavePlayer_12, WavePlayer_13, WavePlayer_14, WavePlayer_15, WavePlayer_16, WavePlayer_17]
  properties {weights [1,1,1,1,1,1]}
 }
 {
  id=SoundClass_0
  children [Attenuation_0]
  properties {soundClassOverride=&Fort_WeaponRanged_Fire_3P_PSM}
 }
 {
  id=SoundClass_32
  children [Attenuation_49]
  properties {soundClassOverride=&Fort_WeaponRanged_Fire_3P_PSM}
 }
 {
  id=SoundClass_35
  children [Attenuation_34]
  properties {soundClassOverride=&Fort_WeaponRanged_Fire_Distant}
 }
 {id=WavePlayer_0, properties {soundWaveAssetPtr=&/BreakActionShotgun_ShootMedium_01}}
 {id=WavePlayer_1, properties {soundWaveAssetPtr=&/BreakActionShotgun_ShootMedium_02}}
 {id=WavePlayer_10, properties {soundWaveAssetPtr=&/BreakActionShotgun_ShootMedium_05}}
 {id=WavePlayer_11, properties {soundWaveAssetPtr=&/BreakActionShotgun_ShootMedium_06}}
 {id=WavePlayer_12, properties {soundWaveAssetPtr=&/BreakActionShotgun_ShootDistant_01}}
 {id=WavePlayer_13, properties {soundWaveAssetPtr=&/BreakActionShotgun_ShootDistant_02}}
 {id=WavePlayer_14, properties {soundWaveAssetPtr=&/BreakActionShotgun_ShootDistant_03}}
 {id=WavePlayer_15, properties {soundWaveAssetPtr=&/BreakActionShotgun_ShootDistant_04}}
 {id=WavePlayer_16, properties {soundWaveAssetPtr=&/BreakActionShotgun_ShootDistant_05}}
 {id=WavePlayer_17, properties {soundWaveAssetPtr=&/BreakActionShotgun_ShootDistant_06}}
 {id=WavePlayer_2, properties {soundWaveAssetPtr=&/BreakActionShotgun_ShootMedium_03}}
 {id=WavePlayer_3, properties {soundWaveAssetPtr=&/BreakActionShotgun_Shoot_01}}
 {id=WavePlayer_4, properties {soundWaveAssetPtr=&/BreakActionShotgun_Shoot_02}}
 {id=WavePlayer_5, properties {soundWaveAssetPtr=&/BreakActionShotgun_Shoot_03}}
 {id=WavePlayer_6, properties {soundWaveAssetPtr=&/BreakActionShotgun_Shoot_04}}
 {id=WavePlayer_7, properties {soundWaveAssetPtr=&/BreakActionShotgun_Shoot_05}}
 {id=WavePlayer_8, properties {soundWaveAssetPtr=&/BreakActionShotgun_Shoot_06}}
 {id=WavePlayer_9, properties {soundWaveAssetPtr=&/BreakActionShotgun_ShootMedium_04}}
]
properties {
 bDebug=true
 maxDistance=27500
 volumeMultiplier=1.1
 concurrencySet [&Weapon_Gunfire_Local_Conc, &Weapon_Gunfire_Global_Conc]
}
