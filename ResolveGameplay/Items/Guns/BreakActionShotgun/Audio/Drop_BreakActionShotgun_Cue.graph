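# Sound cue for the break-action shotgun's drop sound.
# A Random node picks one of three drop wave variations with equal weight.
# The cue uses the shared weapon-drop attenuation and loot-collect sound class, with a boosted volume multiplier.
# It joins the Weapon_Drop_StopOldest concurrency set and also carries an override rule that prevents new sounds.
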SoundCue /ResolveGameplay/Items/Guns/BreakActionShotgun/Audio/Drop_BreakActionShotgun_Cue
root=Random_0
nodes [
 {id=Random_0, children [WavePlayer_7,WavePlayer_8,WavePlayer_9], properties {weights [1,1,1]}}
 {id=WavePlayer_7, properties {soundWaveAssetPtr=&/BreakActionShotgun_Drop_01}}
 {id=WavePlayer_8, properties {soundWaveAssetPtr=&/BreakActionShotgun_Drop_02}}
 {id=WavePlayer_9, properties {soundWaveAssetPtr=&/BreakActionShotgun_Drop_03}}
]
properties {
 attenuationSettings=&Fort_Weapon_Drop_Att
 maxDistance=3000
 soundClassObject=&Fort_Loot_Collect
 volumeMultiplier=1.5
 concurrencyOverrides {resolutionRule=PreventNew}
 concurrencySet [&Weapon_Drop_StopOldest]
}
