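# Sound cue for the dry-fire click a break-action shotgun plays when out of ammo.
# A random node picks one of three equally weighted dry-fire waves, then a modulator varies pitch slightly upward and applies a fixed volume boost.
# The cue routes through a third-person ranged weapon fire sound class and overrides concurrency to allow few instances, stopping the oldest.
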SoundCue /ResolveGameplay/Items/Guns/BreakActionShotgun/Audio/OutofAmmo_BreakActionShotgun_Cue
root=Modulator_0
nodes [
 {id=Modulator_0, children [Random_0], properties {pitchMax=1.2, pitchMin=1, volumeMax=2.2, volumeMin=2.2}}
 {id=Random_0, children [WavePlayer_4,WavePlayer_5,WavePlayer_6], properties {weights [1,1,1]}}
 {id=WavePlayer_4, properties {soundWaveAssetPtr=&/BreakActionShotgun_DryFire_01}}
 {id=WavePlayer_5, properties {soundWaveAssetPtr=&/BreakActionShotgun_DryFire_02}}
 {id=WavePlayer_6, properties {soundWaveAssetPtr=&/BreakActionShotgun_DryFire_03}}
]
properties {
 bOverrideConcurrency=true
 maxDistance=2097152
 soundClassObject=&Fort_WeaponRanged_Fire_3P_PSM
 volumeMultiplier=0.55
 attenuationOverrides {falloffDistance=1600}
 concurrencyOverrides {maxCount=2, resolutionRule=StopOldest}
}
