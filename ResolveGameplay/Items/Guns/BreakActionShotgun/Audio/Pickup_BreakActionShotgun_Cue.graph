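# Sound cue played for picking up the break-action shotgun, using weapon pickup attenuation, the loot collect sound class and pickup concurrency.
# A quality-level node routes its first two tiers and last two tiers to separate mixers, each boosting an equally weighted random pick of three pickup waves.
# The two branches use same-named waves from different paths, and each mixer's second input is empty.
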SoundCue /ResolveGameplay/Items/Guns/BreakActionShotgun/Audio/Pickup_BreakActionShotgun_Cue
root=QualityLevel_0
nodes [
 {id=Mixer_0, children [Random_1,null], properties {inputVolume [1.2,1]}}
 {id=Mixer_1, children [Random_0,null], properties {inputVolume [1.2,1]}}
 {id=QualityLevel_0, children [Mixer_1,Mixer_1,Mixer_0,Mixer_0]}
 {id=Random_0, children [WavePlayer_2,WavePlayer_3,WavePlayer_5], properties {weights [1,1,1]}}
 {id=Random_1, children [WavePlayer_7,WavePlayer_8,WavePlayer_9], properties {weights [1,1,1]}}
 {id=WavePlayer_2, properties {soundWaveAssetPtr=&/BreakActionShotgun_PickUp_01}}
 {id=WavePlayer_3, properties {soundWaveAssetPtr=&/BreakActionShotgun_PickUp_02}}
 {id=WavePlayer_5, properties {soundWaveAssetPtr=&/BreakActionShotgun_PickUp_03}}
 {id=WavePlayer_7, properties {soundWaveAssetPtr=&BreakActionShotgun_PickUp_01}}
 {id=WavePlayer_8, properties {soundWaveAssetPtr=&BreakActionShotgun_PickUp_02}}
 {id=WavePlayer_9, properties {soundWaveAssetPtr=&BreakActionShotgun_PickUp_03}}
]
properties {
 attenuationSettings=&Fort_Weapon_Pickup_Att
 maxDistance=3500
 soundClassObject=&Fort_Loot_Collect
 volumeMultiplier=0.398
 concurrencyOverrides {maxCount=1, resolutionRule=PreventNew}
 concurrencySet [&PickupMaximum_Oldest]
}
