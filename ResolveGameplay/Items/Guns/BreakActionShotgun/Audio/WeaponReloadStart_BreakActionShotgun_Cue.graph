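# Sound cue for the start of the break-action shotgun reload, layering the reload action with a delayed shell drop.
# A mixer plays a random pick of five reload waves, pitched up and boosted by a modulator, alongside a fixed-delay branch.
# That branch waits just over a second, then plays one of six boosted random shell-drop waves.
# The cue uses weapon reload attenuation, the weapon foley sound class and overridden concurrency.
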SoundCue /ResolveGameplay/Items/Guns/BreakActionShotgun/Audio/WeaponReloadStart_BreakActionShotgun_Cue
root=Mixer_0
nodes [
 {id=Delay_0, children [Modulator_1], properties {delayMax=1.2, delayMin=1.2}}
 {id=Mixer_0, children [Modulator_0,Delay_0], properties {inputVolume [1,1]}}
 {id=Modulator_0, children [Random_0], properties {pitchMax=1.1, pitchMin=1.1, volumeMax=1.8, volumeMin=1.8}}
 {id=Modulator_1, children [Random_1], properties {volumeMax=2, volumeMin=2}}
 {id=Random_0, children [WavePlayer_0,WavePlayer_1,WavePlayer_2,WavePlayer_3,WavePlayer_4], properties {weights [1,1,1,1,1]}}
 {
  id=Random_1
  children [WavePlayer_6, WavePlayer_7, WavePlayer_8, WavePlayer_9, WavePlayer_10, WavePlayer_11]
  properties {weights [1,1,1,1,1,1]}
 }
 {id=WavePlayer_0, properties {soundWaveAssetPtr=&/BreakActionShotgun_Reload_01}}
 {id=WavePlayer_1, properties {soundWaveAssetPtr=&/BreakActionShotgun_Reload_05}}
 {id=WavePlayer_10, properties {soundWaveAssetPtr=&/BreakActionShotgun_ReloadShellDrop_05}}
 {id=WavePlayer_11, properties {soundWaveAssetPtr=&/BreakActionShotgun_ReloadShellDrop_06}}
 {id=WavePlayer_2, properties {soundWaveAssetPtr=&/BreakActionShotgun_Reload_04}}
 {id=WavePlayer_3, properties {soundWaveAssetPtr=&/BreakActionShotgun_Reload_03}}
 {id=WavePlayer_4, properties {soundWaveAssetPtr=&/BreakActionShotgun_Reload_02}}
 {id=WavePlayer_6, properties {soundWaveAssetPtr=&/BreakActionShotgun_ReloadShellDrop_01}}
 {id=WavePlayer_7, properties {soundWaveAssetPtr=&/BreakActionShotgun_ReloadShellDrop_02}}
 {id=WavePlayer_8, properties {soundWaveAssetPtr=&/BreakActionShotgun_ReloadShellDrop_03}}
 {id=WavePlayer_9, properties {soundWaveAssetPtr=&/BreakActionShotgun_ReloadShellDrop_04}}
]
properties {
 attenuationSettings=&Fort_Weapon_Reload_Att
 bDebug=true
 bOverrideConcurrency=true
 maxDistance=2250
 soundClassObject=&Fort_WeaponFoley
 volumeMultiplier=0.501
}
